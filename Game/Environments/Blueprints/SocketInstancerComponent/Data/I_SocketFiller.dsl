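; I_SocketFiller is an interface for the SocketInstancerComponent environment tools, judging by its name and folder.
; It declares one function, ReconstructSocketFiller, which returns a boolean and defaults to false.
; The file holds no other logic, and the name suggests implementers rebuild their socket-filling content when it is called.
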(fn ReconstructSocketFiller
  (return false))
