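; Blueprint interface that bridges socket filler objects to editor-side logic for the Socket Instancer Component.
; It declares getters that pull mesh data by property name and a mesh list from a source object.
; It also declares a RunConstruct call, presumably to re-run construction on the implementer.
; The functions are signatures only, with placeholder returns and no logic.
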(fn GetMeshDataFromSource
  (params (Source :type Object%20Reference) (Property :type Name))
  (return false 0))

(fn GetMeshListFromSource
  (params (Source :type Object%20Reference))
  (return false 0))

(fn RunConstruct)
