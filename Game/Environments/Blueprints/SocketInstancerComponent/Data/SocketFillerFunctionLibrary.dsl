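; SocketFillerFunctionLibrary is a Blueprint function library with a helper that rebuilds socket filler components on an actor.
; ReconstructAllSocketFillers checks the calling actor is valid, synchronously loads a filler class from a local soft class reference and casts it to an actor component class.
; It then gets all components of that class on the actor and sends each the ISocketFiller ReconstructSocketFiller interface message.
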(fn ReconstructAllSocketFillers
  (params (CallingActor :type Actor%20Object%20Reference))
  (Utilities|IsValid CallingActor
    (:Is%20Valid
      (bind _returnvalue (Utilities|LoadClassAssetBlocking (LocalVariables|Getl_socketFillerClassRef)))
      (bind _asactor_component (Utilities|Casting|CastToActorComponentClass _returnvalue))
      (for _array_element (Actor|GetComponentsByClass CallingActor _asactor_component)
        (Class|ISocketFiller|ReconstructSocketFiller(Message) _array_element)))))
